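; Gameplay cue burst notify that adds world effects when the Shockwave Mace's shockwave bounce cue fires on its target.
; Behind an AllowJanus gate, it reads the target's location and spawns an FN_RadialForce physics field actor there if the r.PhysicsField.EnableField console variable is on.
; It also moves the first burst particle component to slightly below that location.
; The Nope path does nothing.
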(fn OnBurst(EitherVFX-preferthisversion)
  (params (MyTarget :type Actor%20Object%20Reference) (Parameters :type Gameplay%20Cue%20Parameters%20Structure%20%28by%20ref%29) (ParticleComponents :type Array%20of%20FXSystem%20Component%20Object%20References) (AudioComponents :type Array%20of%20Audio%20Component%20Object%20References) (BurstCameraShakeInstance :type Camera%20Shake%20Base%20Object%20Reference) (BurstDecalInstance :type Decal%20Actor%20Object%20Reference))
  (bind _returnvalue (Transformation|GetActorLocation MyTarget))
  ;[1 1 1] Add RadialForce
  (Utilities|AllowJanus
    (:Allow
      (bind _returnvalue_2 (Development|GetConsoleVariableIntValue "r.PhysicsField.EnableField"))
      (if (== _returnvalue_2 1)
        (Game|SpawnActorfromClass 0 "/Game/Effects/Niagara/Blueprints/PhsyicsFields/FN_RadialForce.FN_RadialForce_C" (Math|Transform|MakeTransform _returnvalue) "Undefined" "MultiplyWithRoot" 0 0.2 0.6 1.2e+03 1.0 0.0 0)
        (else
          _))
      (Transformation|SetWorldLocation (Utilities|Array|Get(acopy) ParticleComponents) (+ _returnvalue "0,0,-60.000000") false false)
      (return))
    (:Nope
      _)))
