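; Burst gameplay cue for the Shockwave Mace slam that places its impact effect on the ground in front of the target actor.
; A downward line trace ahead of the actor supplies location and surface normal, falling back to actor location and world up on a miss.
; Behind an AllowJanus gate, it moves a burst particle component to that offset, surface-aligned transform, first spawning a radial-force field actor there if the physics field console variable is enabled.
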(fn OnBurst(EitherVFX-preferthisversion)
  (params (MyTarget :type Actor%20Object%20Reference) (Parameters :type Gameplay%20Cue%20Parameters%20Structure%20%28by%20ref%29) (ParticleComponents :type Array%20of%20FXSystem%20Component%20Object%20References) (AudioComponents :type Array%20of%20Audio%20Component%20Object%20References) (BurstCameraShakeInstance :type Camera%20Shake%20Base%20Object%20Reference) (BurstDecalInstance :type Decal%20Actor%20Object%20Reference))
  (bind _mytarget (Variables|GetMyTarget))
  (bind _rootcomponent (Variables|/Script/Engine.Actor|GetRootComponent _mytarget))
  (bind _returnvalue (Transformation|GetRightVector _rootcomponent))
  (bind _mytarget_2 (Variables|GetMyTarget))
  (bind _returnvalue_2 (Transformation|GetActorLocation _mytarget_2))
  (bind _returnvalue_3 (Transformation|GetForwardVector _rootcomponent))
  (bind _returnvalue_4 (* _returnvalue 16.0))
  (bind (_outhit _returnvalue_5) (Collision|LineTraceByChannel MyTarget (+ (+ (Transformation|GetActorLocation MyTarget) (* (Transformation|GetForwardVector (Variables|/Script/Engine.Actor|GetRootComponent MyTarget)) 1.5e+02)) "0,0,50.000000") (- (+ (Transformation|GetActorLocation MyTarget) (* (Transformation|GetForwardVector (Variables|/Script/Engine.Actor|GetRootComponent MyTarget)) 1.5e+02)) "0,0,200.000000") "TraceTypeQuery1" false (Utilities|Array|MakeArray MyTarget) "ForOneFrame"))
  ;[1 1 1] Add RadialForce
  (Utilities|AllowJanus
    (:Allow
      (bind _returnvalue_6 (Development|GetConsoleVariableIntValue "r.PhysicsField.EnableField"))
      (if (== _returnvalue_6 1)
        (Game|SpawnActorfromClass 0 "/Game/Effects/Niagara/Blueprints/PhsyicsFields/FN_RadialForce.FN_RadialForce_C" (Math|Transform|MakeTransform (Utilities|Operators|Add (Math|Vector|SelectVector (.Location (Collision|BreakHitResult _outhit)) _returnvalue_2 _returnvalue_5) "0,0,16.000000" _returnvalue_4) (Math|Rotator|MakeRotfromZX (Math|Vector|SelectVector (Math|Vector|SelectVector (.ImpactNormal (Collision|BreakHitResult _outhit)) "0, 0,1.000000" _returnvalue_5) "0, 0,1.000000" (Utilities|Operators|NotEqual(!=) :A (Math|Vector|SelectVector (.ImpactNormal (Collision|BreakHitResult _outhit)) "0, 0,1.000000" _returnvalue_5))) (Math|Vector|ProjectVectorontoPlane _returnvalue_3 (Math|Vector|SelectVector (Math|Vector|SelectVector (.ImpactNormal (Collision|BreakHitResult _outhit)) "0, 0,1.000000" _returnvalue_5) "0, 0,1.000000" (Utilities|Operators|NotEqual(!=) :A (Math|Vector|SelectVector (.ImpactNormal (Collision|BreakHitResult _outhit)) "0, 0,1.000000" _returnvalue_5)))))) "Undefined" "MultiplyWithRoot" 0 0.2 0.6 1.2e+03 1.0 0.0 0)
        (else
          _))
      (Transformation|SetWorldTransform (Utilities|Array|Get(acopy) (Variables|GetParticleComponents)) (Math|Transform|MakeTransform (Utilities|Operators|Add (Math|Vector|SelectVector (.Location (Collision|BreakHitResult _outhit)) _returnvalue_2 _returnvalue_5) "0,0,16.000000" _returnvalue_4) (Math|Rotator|MakeRotfromZX (Math|Vector|SelectVector (Math|Vector|SelectVector (.ImpactNormal (Collision|BreakHitResult _outhit)) "0, 0,1.000000" _returnvalue_5) "0, 0,1.000000" (Utilities|Operators|NotEqual(!=) :A (Math|Vector|SelectVector (.ImpactNormal (Collision|BreakHitResult _outhit)) "0, 0,1.000000" _returnvalue_5))) (Math|Vector|ProjectVectorontoPlane _returnvalue_3 (Math|Vector|SelectVector (Math|Vector|SelectVector (.ImpactNormal (Collision|BreakHitResult _outhit)) "0, 0,1.000000" _returnvalue_5) "0, 0,1.000000" (Utilities|Operators|NotEqual(!=) :A (Math|Vector|SelectVector (.ImpactNormal (Collision|BreakHitResult _outhit)) "0, 0,1.000000" _returnvalue_5)))))) false false)
      (return))
    (:Nope
      _)))
